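; CherrySmoke DualRevolvers pawn component that tracks aerial movement state: air dodge, double jump and rebound.
; Custom events set the flags, with air dodge and double jump excluding each other and both clearing the rebound flag.
; Starting a rebound only raises its own flag, while ending a jump or dodge clears all three.
; GetAnimData returns the three booleans together, evidently for animation use.
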(event Custom|StartAirDodge
  (Variables|Default|SetAirDodgeActive true)
  (Variables|Default|SetDoubleJumpActive false)
  (Variables|Default|SetDidRebound false))

(event Custom|StartDoubleJump
  (Variables|Default|SetAirDodgeActive false)
  (Variables|Default|SetDoubleJumpActive true)
  (Variables|Default|SetDidRebound false))

(event Custom|StartRebound
  (Variables|Default|SetDidRebound true))

(event Custom|EndJumpOrAirDodge
  (Variables|Default|SetAirDodgeActive false)
  (Variables|Default|SetDoubleJumpActive false)
  (Variables|Default|SetDidRebound false))

(fn GetAnimData
  (return (Variables|Default|GetDidRebound) (Variables|Default|GetAirDodgeActive) (Variables|Default|GetDoubleJumpActive)))
